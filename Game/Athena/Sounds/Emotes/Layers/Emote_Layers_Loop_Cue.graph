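# Sound cue for the music of the Layers emote in Athena, playing an intro that leads into a looping track.
# A Mixer starts both branches together: the intro after a very brief Delay, and the loop after a longer Delay, with an Enveloper giving it a quick fade-in.
# It uses emote attenuation and music sound class, stop-oldest concurrency override, and a send to the music audio-analysis submix.
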SoundCue /Game/Athena/Sounds/Emotes/Layers/Emote_Layers_Loop_Cue
root=Mixer_0
nodes [
 {id=Delay_0, children [Enveloper_1], properties {delayMax=0.57, delayMin=0.57}}
 {id=Delay_1, children [WavePlayer_3], properties {delayMax=0.02, delayMin=0.02}}
 {
  id=Enveloper_1
  children [WavePlayer_0]
  properties {
   pitchCurve {editorCurveData {keys [{time=0}]}}
   volumeCurve {editorCurveData {keys [<at 0 0>,<at 0.0182 0.3155>,<at 0.0275 0.6291>,<at 0.0338 1>]}}
  }
 }
 {id=Mixer_0, children [Delay_1,Delay_0], properties {inputVolume [1,1]}}
 {id=WavePlayer_0, properties {bLooping=true, soundWaveAssetPtr=&/S19_Emote_Layers_Loop}}
 {id=WavePlayer_3, properties {soundWaveAssetPtr=&/S19_Emote_Layers_Intro}}
]
properties {
 attenuationSettings=&Fort_Emotes_Athena_Att
 bOverrideConcurrency=true
 maxDistance=1600
 soundClassObject=/Game/Sounds/FortSoundClassesAndModes/Fort_Music_Emote_1P_PSM
 volumeMultiplier=1
 concurrencyOverrides {maxCount=3, resolutionRule=StopOldest}
 concurrencySet [/Game/Sounds/Concurrency_Rules/Athena_EmoteMusic_Conc]
 soundSubmixSends [{minSendLevel=0, soundSubmix=&Fortnite_AudioAnalysis_Music}]
}
